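# Material function that draws a number, capped at 999, as three digits, apparently for a weapon's ADS distance readout.
# Frac and rounding math extracts each digit, picked per UV column to offset into a digit-strip texture.
# UVs are padded about the center, and masks with a custom rotator can tilt side digits into a curve.
# Sampling uses original-UV derivatives; outputs include greyscale, color, masks, UVs and debug values.
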MaterialFunction /JethroWeaponMods/Shared_Material/MF_ADS_Distance_Jethro
graph {nodes [
  Constant2Vector() {g=1, r=0.35}
  Fmod()
  n1 = FunctionInput(Preview=Constant() {r=100.999}) {bUsePreviewValueAsDefault=true, inputName=Number}
  n2 = Constant() {r=999}
  n3 = If(A=@n1, B=@n2, "A > B"=@n2, "A == B"=@n2, "A < B"=@n1)
  n4 = Add(A=@n3, B=Constant() {desc=fix for strange bug, r=0.01})
  n5 = If(A=@n3, "A > B"=@n4, "A == B"=@n3, "A < B"=@n4)
  n6 = Divide(A=@n5) {constB=100}
  Fmod(A=@n6, B=Constant() {r=1})
  n7 = FunctionInput(Preview=TextureCoordinate() {coordinateIndex=2}) {bUsePreviewValueAsDefault=true, inputName=UVs, inputType=FunctionInput_Vector2, sortPriority=1}
  n8 = ScalarParameter() {defaultValue=1, group=Numbers, parameterName=u_NumberPaddingScale}
  n9 = ScalarParameter() {defaultValue=1, group=Numbers, parameterName=v_NumberPaddingScale}
  n10 = MaterialFunctionCall(
   UVs=@n7
   Texture Scale=FunctionInput(Preview=AppendVector(A=@n8, B=@n9)) {bUsePreviewValueAsDefault=true, inputName=UVPadding, inputType=FunctionInput_Vector2, sortPriority=1}
  ) {materialFunction=&ScaleUVsByCenter}
  n11 = Constant2Vector() {r=1}
  n12 = Multiply(A=Subtract(A=DotProduct(A=@n7, B=@n11)) {constB=0.5}) {constB=256}
  n13 = Saturate(None=@n12)
  n14 = LinearInterpolate(Alpha=@n13) {constA=0.2, constB=0.8}
  n15 = Constant() {r=0.5}
  n16 = ScalarParameter() {defaultValue=0.02, group=Numbers, parameterName=Curve_Rotation}
  n17 = Multiply(A=@n16) {constB=-1}
  n18 = OneMinus(None=@n12)
  n19 = StaticBool()
  n20 = NamedRerouteDeclaration(None=MaterialFunctionCall!(A=@n10.UVs, B=Constant2Vector() {g=0.5, r=0.5}, Bounds=Constant2Vector() {g=1, r=0.345}, Edge Falloff=Constant() {r=0.001}) {materialFunction=/Engine/Functions/Engine_MaterialFunctions01/Gradient/BoxMask-2D}.Result) {name=Middle Mask, nodeColor=<rgba 0.1016 1 0 1>}
  n21 = Add(
   A=StaticSwitch(
    True=@n13
    False=Saturate(None=@n18)
    Value=FunctionInput(Preview=@n19) {bUsePreviewValueAsDefault=true, inputName=Left?, inputType=FunctionInput_StaticBool, sortPriority=3}
   )
   B=@n20
  )
  n22 = FunctionInput(Preview=Constant() {r=1}) {bUsePreviewValueAsDefault=true, description=Feed 0 if side number, inputName=Side Set, inputType=FunctionInput_Scalar, sortPriority=2}
  n23 = LinearInterpolate(
   A=MaterialFunctionCall(
    UVs=@n10.UVs
    Rotation Center=AppendVector(A=@n14, B=@n15)
    "Rotation Angle (0-1)"=LinearInterpolate(A=@n16, B=@n17, Alpha=@n13) {constAlpha=1}
   ) {materialFunction=&CustomRotator}.Rotated Values
   B=@n10.UVs
   Alpha=Saturate(None=LinearInterpolate(A=@n21, B=@n20, Alpha=@n22))
  )
  n24 = Multiply(A=ComponentMask(None=@n23) {r=true}, B=Constant() {desc=Number of digits, r=3}) {constB=5}
  n25 = Ceil(None=@n24)
  n26 = Constant() {r=2}
  n27 = Constant() {r=3}
  n28 = Divide(A=Subtract(A=@n5, B=Frac(None=@n5))) {constB=10}
  n29 = Round(None=Multiply(A=Frac(None=@n28)) {constB=10})
  n30 = If(A=@n25, B=@n27, "A > B"=@n27, "A == B"=@n29, "A < B"=@n29)
  n31 = Divide(A=@n5) {constB=10}
  n32 = Divide(A=Subtract(A=@n31, B=Frac(None=@n31))) {constB=10}
  n33 = Round(None=Multiply(A=Frac(None=@n32)) {constB=10})
  n34 = Frac(None=@n6)
  n35 = Divide(A=Subtract(A=@n6, B=@n34)) {constB=10}
  n36 = Round(None=Multiply(A=Frac(None=@n35)) {constB=10})
  n37 = Add(
   A=Frac(None=@n24)
   B=If(A=@n25, B=Constant() {r=1}, "A > B"=If(A=@n25, B=@n26, "A > B"=@n30, "A == B"=@n33, "A < B"=@n33), "A == B"=@n36, "A < B"=@n36)
  )
  n38 = ScalarParameter() {defaultValue=11, group=Numbers, parameterName=Numbers of Digits On Texture}
  n39 = ComponentMask(None=@n23) {g=true}
  n40 = Add(A=@n39, B=ScalarParameter() {defaultValue=0.05, group=Numbers, parameterName=Middle_Offset})
  n41 = NamedRerouteUsage()
  n42 = TextureSampleParameter2D!(
   UVs=AppendVector(A=Divide(A=@n37, B=@n38) {constB=11}, B=LinearInterpolate(A=@n39, B=@n40, Alpha=@n41) {constAlpha=0})
   "DDX(UVs)"=DDX(Value=@n7)
   "DDY(UVs)"=DDY(Value=@n7)
   Apply View MipBias
  ) {
   group=Numbers
   mipValueMode=TMVM_Derivative
   parameterName=NumbersTexture
   samplerType=SAMPLERTYPE_Masks
   texture=/Game/Athena/Events/Junior/T_Numbers2J_M
  }
  FunctionOutput(None=Multiply(A=@n42.R, B=@n10.0-1 mask)) {outputName=NumbersGreyscale}
  FunctionOutput(None=@n10.0-1 mask) {outputName=0-1 Mask, sortPriority=2}
  FunctionOutput(None=@n42.RGB) {outputName=NumbersColor, sortPriority=1}
  n43 = Add(A=TextureCoordinate(), B=Constant2Vector() {g=0.25})
  FunctionOutput(None=Multiply(
    A=Constant3Vector() {constant=<rgba 1 0.0202 0 0>}
    B=MaterialFunctionCall(Number=@n34, UVs=@n43) {materialFunction=&DebugScalarValues}.Result
   )) {outputName=debug1, sortPriority=434}
  FunctionOutput(None=@n10.UVs) {outputName=UVS, sortPriority=8}
  FunctionOutput(None=@n10.0-1 mask) {outputName=Mask, sortPriority=9}
  n44 = TextureCoordinate()
  n45 = Constant2Vector() {g=-0.25}
  FunctionOutput(None=MaterialFunctionCall(Number=@n6, UVs=Add(A=@n44, B=@n45)) {materialFunction=&DebugScalarValues}.Result) {outputName=debug1_1, sortPriority=434}
  ScalarParameter() {defaultValue=2, group=Numbers, parameterName=Digit_2}
  ScalarParameter() {defaultValue=1, group=Numbers, parameterName=Digit_1}
  ScalarParameter() {defaultValue=3, group=Numbers, parameterName=Digit_3}
 ]}
