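# Post-process material for Jethro weapon optics, drawing the scope view, reticle and HUD readouts during ADS.
# MF_FirstPersonScope supplies masks for zoom refraction, alongside a vignette and optional SM6 custom-HLSL radial DOF blur.
# The reticle supports SDF, fisheye, static noise, toggled colors and spread-driven mip-bias/multi-sample blur.
# Code-fed parameters such as TargetingWeight and SpreadRatio animate it, with rangefinder digits, overheat gauge, scanlines and flicker also built in.
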Material /JethroWeaponMods/Shared_Material/PPM_Jethro_OpticsV2_1
graph {
 properties {materialDomain=MD_PostProcess, pixelDepthOffsetMode=PDOM_Legacy, usedShadingModels=Default Lit}
 nodes [
  n1 = NamedRerouteUsage()
  n2 = Add(A=NamedRerouteUsage(), B=@n1)
  n3 = Saturate(None=InverseLinearInterpolate(Value=NamedRerouteUsage()) {constA=100, constB=35})
  n4 = MaterialFunctionCall(
   UVs=MaterialFunctionCall(
    UVs=StaticSwitchParameter(True=@n2, False=@n1) {group=Reticle Fisheye, parameterName=UseCrosshairFisheyeDistortion, sortPriority=34}
    Texture Scale=LinearInterpolate(Alpha=@n3) {constA=0.1}
   ) {materialFunction=&ScaleUVsByCenter}.UVs
   Texture Scale=ScalarParameter() {defaultValue=0.25, group=Crosshair, parameterName=Crosshair_Scale, sliderMax=1}
  ) {materialFunction=&ScaleUVsByCenter}
  n5 = Multiply(A=TextureCoordinate(), B=ScalarParameter() {defaultValue=1, group=Crosshair, parameterName=StaticNoise_Scale})
  n6 = TextureSample(UVs=Panner(Coordinate=@n5) {speedX=0.7, speedY=1}) {
   samplerSource=SSM_Wrap_WorldGroupSettings
   samplerType=SAMPLERTYPE_LinearGrayscale
   texture=/Game/Effects/Fort_Effects/Textures/Tile/T_WhiteNoise_02_M
  }
  n7 = TextureSample(UVs=Panner(Coordinate=@n5) {speedX=-0.5, speedY=-2}) {
   samplerSource=SSM_Wrap_WorldGroupSettings
   samplerType=SAMPLERTYPE_LinearGrayscale
   texture=&T_WhiteNoise_02_M
  }
  n8 = Multiply(A=@n6.R, B=@n7.R)
  n9 = Multiply(A=@n8, B=ScalarParameter() {defaultValue=0.0042, group=Crosshair, parameterName=StaticNoise_Str})
  n10 = Multiply(A=@n8, B=ScalarParameter() {defaultValue=0.0012, group=Crosshair, parameterName=StaticNoise_Str_low})
  n11 = NamedRerouteUsage()
  n12 = NamedRerouteUsage()
  n13 = ScalarParameter() {desc="Determines safe zone where static can't occur (Reticule perfered)", group=Crosshair, parameterName=Static_safety_radius}
  n14 = Constant() {r=256}
  Add(
   A=@n4.UVs
   B=LinearInterpolate(
    A=LinearInterpolate(A=@n9, B=@n10, Alpha=@n11) {constB=0}
    Alpha=MaterialFunctionCall(UVs=@n12, Radius=@n13, Density=@n14) {materialFunction=&RadialGradientExponential}.RadialGradientExponential
   ) {constB=0}
  )
  ComponentMask() {g=true, r=true}
  Constant() {desc=.5 got us a good angle for the numbers, r=0.5}
  n15 = Multiply(A=Time(), B=ScalarParameter() {defaultValue=0.25, parameterName=DebugSpeed})
  Divide(A=Add(A=Sine(None=@n15)))
  n16 = ComponentMask(None=VectorParameter() {parameterName=PlayerVelocity, defaultValue=<rgba 0 0 0 1>}.RGB) {b=true, g=true, r=true}
  Divide(
   A=Subtract(A=Length(None=@n16) {is=class MaterialExpressionLength}, B=ScalarParameter() {defaultValue=1, parameterName=MinVelocityCrossHairBlur})
   B=ScalarParameter() {defaultValue=100, parameterName=MaxVelocityCrossHairBlur}
  )
  n17 = Constant2Vector() {g=0.5, r=0.5}
  Divide(A=ComponentMask(None=SceneTexture(UVs=@n17) {bFiltered=true, sceneTextureId=PPI_SceneDepth}.Color) {r=true}) {constB=100, desc=material distance check}
  n18 = MaterialFunctionCall!(
   Plane Depth Offset=ScalarParameter() {group=Optics Masks, parameterName=Offset}
   Scope Eye Mask Hardness=Constant()
   Scale=ScalarParameter() {defaultValue=2.2, group=Optics Masks, parameterName=Scope Scale}
   Scope Target Point WS
   Scope Eye Point WS
   Up Vector
   In
  ) {materialFunction=/Game/Packages/Fortress_SharedMaterials/Base_Material_Functions/MF_FirstPersonScope}
  InverseLinearInterpolate(
   A=ScalarParameter() {defaultValue=225, group=ADS Transition, parameterName=BlendDistanceStart}
   B=ScalarParameter() {defaultValue=50, group=ADS Transition, parameterName=BlendDistanceEnd}
   Value=@n18.Camera - Eye Point Distance
  )
  n19 = Power(Base=SmoothStep(
    Min=ScalarParameter() {group=Zoom, parameterName=Refraction Remap Min, sliderMax=1}
    Max=ScalarParameter() {defaultValue=0.2, group=Zoom, parameterName=Refraction Remap Max, sliderMax=1}
    Value=NamedRerouteUsage()
   )) {constExponent=0.5}
  n20 = NamedRerouteUsage()
  n21 = LinearInterpolate(
   A=ScalarParameter() {defaultValue=1, group=Zoom, parameterName=Optical Zoom Min, sliderMax=2, sliderMin=1}
   B=ScalarParameter() {defaultValue=1, group=Zoom, parameterName=Optical Zoom Max, sliderMax=2, sliderMin=1}
   Alpha=Multiply(A=NamedRerouteDeclaration(None=@n19) {name=Refraction Debug, nodeColor=<rgba 0.176 1 0 1>}, B=Power(Base=@n20) {constExponent=4})
  ) {constA=1}
  n22 = NamedRerouteUsage()
  n23 = Constant() {r=1}
  LinearInterpolate(B=@n21, Alpha=Multiply(A=SphereMask(A=@n22, B=@n23) {attenuationRadius=0.2, hardnessPercent=0}, B=NamedRerouteUsage())) {constA=1}
  LinearInterpolate(Alpha=Custom!(None) {code=View.GPUSceneViewId, showCode=true})
  n24 = NamedRerouteUsage()
  Max(A=@n24, B=NamedRerouteUsage())
  n25 = ScalarParameter() {defaultValue=0.5, group=Reticle Fisheye, parameterName=Reticle Fisheye Center X, sortPriority=50}
  n26 = ScalarParameter() {defaultValue=0.5, group=Reticle Fisheye, parameterName=Reticle Fisheye Center Y, sortPriority=51}
  NamedRerouteDeclaration(None=MaterialFunctionCall!(
    Coords
    Center=AppendVector(A=@n25, B=@n26)
    Amout=ScalarParameter() {defaultValue=0.2, group=Reticle Fisheye, parameterName=Reticle Fisheye Amount, sortPriority=100}
    Exponent=ScalarParameter() {defaultValue=3, group=Reticle Fisheye, parameterName=Reticle Fisheye Exponent, sortPriority=200}
   ) {materialFunction=/Game/Weapons/FORT_Rifles/SniperScope/MF_Fisheye_Distortion}.Result) {name=Fisheye Distortion, nodeColor=<rgba 0 0.577 1 1>}
  n27 = Subtract(A=NamedRerouteUsage()) {constB=0.5}
  n28 = ScalarParameter() {defaultValue=3, group=Vignette, parameterName=Rec Vignette X}
  n29 = ScalarParameter() {defaultValue=3, group=Vignette, parameterName=Rec Vignette Y}
  n30 = ScalarParameter() {defaultValue=0.001, desc=r, group=Vignette, parameterName=Rectangular Vignette Radius}
  n31 = Add(A=Subtract(A=Abs(None=@n27), B=AppendVector(A=@n28, B=@n29)), B=@n30)
  n32 = ComponentMask(None=@n31) {desc=q.x, r=true}
  n33 = ComponentMask(None=@n31) {desc=q.y, g=true}
  n34 = Add(
   A=Min(A=Max(A=@n32, B=@n33)) {constB=0}
   B=MaterialFunctionCall(Vector 2=Max(A=@n31) {constB=0, desc="max(q,0)"}) {materialFunction=/Engine/Functions/Engine_MaterialFunctions02/Utility/VectorLength}.V2 Length
  )
  n35 = SphereMask(
   A=Saturate(None=Subtract(A=@n34, B=@n30))
   B=Constant()
   Radius=ScalarParameter() {defaultValue=0.5, group=Vignette, parameterName=Rec Vignette Radius}
   Hardness=ScalarParameter() {group=Vignette, parameterName=Rec Vignette Hardness}
  )
  n36 = SphereMask(
   A=Subtract(A=NamedRerouteUsage()) {constB=0.5}
   B=Constant()
   Radius=LinearInterpolate(
    A=ScalarParameter() {defaultValue=15, group=Vignette, parameterName=Fade Center Scale Start}
    B=ScalarParameter() {defaultValue=4, group=Vignette, parameterName=Fade Center Scale}
    Alpha=NamedRerouteUsage()
   )
   Hardness=ScalarParameter() {group=Vignette, parameterName=Fade Center Hardness}
  )
  n37 = ScalarParameter() {defaultValue=1, group=Vignette, parameterName=Fade Center Power}
  n38 = LinearInterpolate(A=ScalarParameter() {group=Vignette, parameterName=Edge Darkness}, Alpha=Saturate(None=Power(Base=@n36, Exp=@n37))) {constA=25}
  n39 = OneMinus(None=NamedRerouteUsage())
  n40 = NamedRerouteUsage()
  NamedRerouteDeclaration(None=LinearInterpolate(B=StaticSwitchParameter(True=@n35, False=@n38) {group=Vignette, parameterName=Use Rectangular Vignette}, Alpha=Multiply(A=@n39, B=@n40)) {constA=1}) {name=Vignette, nodeColor=<rgba 0.2 0.2 0.2 1>}
  n41 = NamedRerouteUsage()
  n42 = NamedRerouteUsage()
  n43 = NamedRerouteUsage()
  n44 = NamedRerouteUsage()
  n45 = ComponentMask(None=LinearInterpolate(A=LinearInterpolate(A=@n41, B=@n42, Alpha=@n43), B=LinearInterpolate(A=@n41, B=@n42, Alpha=@n44), Alpha=NamedRerouteUsage())) {b=true, g=true, r=true}
  n46 = Multiply(A=NamedRerouteUsage(), B=NamedRerouteUsage())
  n47 = LinearInterpolate(A=NamedRerouteUsage(), B=ComponentMask(None=NamedRerouteUsage()) {b=true, g=true, r=true}, Alpha=NamedRerouteUsage())
  n48 = Add(A=Multiply(A=@n45, B=@n46), B=@n47)
  n49 = NamedRerouteUsage()
  n50 = LinearInterpolate(A=@n47, B=@n45, Alpha=@n46) {constA=1}
  n51 = Add(A=NamedRerouteUsage(), B=NamedRerouteUsage())
  n52 = NamedRerouteUsage()
  NamedRerouteDeclaration(None=Add(A=LinearInterpolate(A=@n48, B=@n50, Alpha=@n49), B=Multiply(A=@n51, B=@n52))) {name=Crosshair, nodeColor=<rgba 0.6641 1 0 1>}
  n53 = TextureSampleParameter2D(UVs=Saturate(None=NamedRerouteDeclaration(None=@n18.Scope Eye UV) {name=Scope Eye UV, nodeColor=<rgba 0.7813 0 1 1>})) {
   group=Optics Masks
   parameterName=Mask Texture
   samplerType=SAMPLERTYPE_Masks
   texture=/Game/Weapons/FORT_Rifles/SniperScope/T_HolographicMask
  }
  n54 = NamedRerouteDeclaration(None=@n18.Scope Eye Mask) {name=Refraction Mask, nodeColor=<rgba 0 1 0.4609 1>}
  NamedRerouteDeclaration(None=StaticSwitchParameter(True=@n53.R, False=SmoothStep(Value=@n54) {constMax=0.01}) {group=Optics Masks, parameterName=UseMaskTexture, sortPriority=1}) {name=Optics Mask, nodeColor=<rgba 1 0.75 0 1>}
  n55 = Multiply(A=NamedRerouteUsage(), B=NamedRerouteUsage())
  n56 = Multiply(A=ComponentMask(None=@n55) {r=true}, B=ComponentMask(None=NamedRerouteUsage()) {a=true})
  n57 = ComponentMask(None=@n55) {g=true}
  n58 = Multiply(A=ComponentMask(None=@n55) {a=true}, B=ScalarParameter() {defaultValue=1, group=Crosshair, parameterName=glare_multi})
  n59 = Multiply(A=Saturate(None=Add(A=@n57, B=@n58)), B=ComponentMask(None=NamedRerouteUsage()) {a=true})
  NamedRerouteDeclaration(None=Max(
    A=NamedRerouteDeclaration(None=@n56) {name=Reticle Mask Core, nodeColor=<rgba 0 1 0.4141 1>}
    B=NamedRerouteDeclaration(None=@n59) {name=Reticle Mask Glow, nodeColor=<rgba 0.1016 0 1 1>}
   )) {name=Reticle Mask, nodeColor=<rgba 0 0.5078 1 1>}
  NamedRerouteDeclaration(None=StaticSwitchParameter(True=@n53.G, False=@n54) {group=Optics Masks, parameterName=UseMaskTexture, sortPriority=1}) {name=Optics Mask Refraction, nodeColor=<rgba 0 1 0.3414 1>}
  NamedRerouteDeclaration(None=@n53.B) {name=Optics Mask Blur, nodeColor=<rgba 0.1016 0 1 1>}
  n60 = NamedRerouteUsage()
  n61 = LinearInterpolate(
   A=ComponentMask(None=SceneTexture() {sceneTextureId=PPI_PostProcessInput0}.Color) {b=true, g=true, r=true}
   B=Constant3Vector() {constant=<rgba 0 0 1 0>}
   Alpha=OneMinus(None=NamedRerouteUsage())
  )
  n62 = Constant3Vector() {constant=<rgba 1 0 0 0>}
  n63 = NamedRerouteUsage()
  n64 = Add(
   A=LinearInterpolate(A=LinearInterpolate(A=@n61, B=@n62, Alpha=@n63), B=Constant3Vector(), Alpha=@n24)
   B=Multiply(A=NamedRerouteUsage(), B=Constant3Vector() {constant=<rgba 0 1 0 0>})
  )
  NamedRerouteDeclaration(None=StaticSwitchParameter(True=NamedRerouteUsage(), False=StaticSwitchParameter(True=@n60, False=@n64) {group=Debug, parameterName=Debug DOF}) {group=Debug, parameterName=Debug Reticle}) {name=DebugView, nodeColor=<rgba 0 1 0.3438 1>}
  n65 = ComponentMask(None=@n55) {b=true}
  n66 = NamedRerouteUsage()
  n67 = NamedRerouteUsage()
  NamedRerouteDeclaration(None=Multiply(A=Multiply(A=@n65, B=@n66), B=ComponentMask(None=@n67) {a=true})) {name=Reticle Secondary Mask, nodeColor=<rgba 0 1 0.5781 1>}
  n68 = VectorParameter() {group=Crosshair, parameterName=Crosshair_GlowColor, defaultValue=<rgba 0 0 0 1>}
  n69 = AppendVector(A=@n68.RGB, B=@n68.RGB)
  n70 = VectorParameter() {group=Toggleable Reticle, parameterName=Crosshair_GlowColorToggled, defaultValue=<rgba 0 0 0 1>}
  n71 = AppendVector(A=@n70.RGB, B=@n70.RGB)
  n72 = ScalarParameter() {parameterName=ToggleThermalMask}
  NamedRerouteDeclaration(None=StaticSwitchParameter(True=LinearInterpolate(A=@n69, B=@n71, Alpha=@n72), False=@n69) {group=Toggleable Reticle, parameterName=Enable Toggleable Reticle, sortPriority=5}) {name=Crosshair Glow Color, nodeColor=<rgba 1 0.4219 0 1>}
  n73 = VectorParameter() {group=Crosshair, parameterName=Crosshair_CoreColor, defaultValue=<rgba 0 0 0 1>}
  n74 = AppendVector(A=@n73.RGB, B=@n73.RGB)
  n75 = VectorParameter() {group=Toggleable Reticle, parameterName=Crosshair_CoreColorToggled, defaultValue=<rgba 0 0 0 1>}
  n76 = AppendVector(A=@n75.RGB, B=@n75.RGB)
  n77 = ScalarParameter() {parameterName=ToggleThermalMask}
  NamedRerouteDeclaration(None=StaticSwitchParameter(True=LinearInterpolate(A=@n74, B=@n76, Alpha=@n77), False=@n74) {group=Toggleable Reticle, parameterName=Enable Toggleable Reticle, sortPriority=5}) {name=Crosshair Core Color, nodeColor=<rgba 1 0.4219 0 1>}
  n78 = VectorParameter() {group=Crosshair, parameterName=Crosshair_SecondaryColor, defaultValue=<rgba 0 0 0 1>}
  n79 = AppendVector(A=@n78.RGB, B=@n78.RGB)
  n80 = VectorParameter() {group=Toggleable Reticle, parameterName=Crosshair_SecondaryColorToggled, defaultValue=<rgba 0 0 0 1>}
  n81 = AppendVector(A=@n80.RGB, B=@n80.RGB)
  n82 = ScalarParameter() {parameterName=ToggleThermalMask}
  NamedRerouteDeclaration(None=StaticSwitchParameter(True=LinearInterpolate(A=@n79, B=@n81, Alpha=@n82), False=@n79) {group=Toggleable Reticle, parameterName=Enable Toggleable Reticle, sortPriority=5}) {name=Crosshair Secondary Color, nodeColor=<rgba 1 0.4219 0 1>}
  n83 = MaterialFunctionCall(
   X=ScalarParameter() {defaultValue=1, group=Crosshair, parameterName=Reticle Opacity, sliderMax=1}
   Y=ScalarParameter() {group=Crosshair, parameterName=Core/Glow Hard Edge, sliderMax=1}
  ) {materialFunction=&MakeFloat4}
  n84 = ScalarParameter() {defaultValue=1, group=Toggleable Reticle, parameterName=Reticle Opacity Toggled, sliderMax=1}
  n85 = ScalarParameter() {group=Toggleable Reticle, parameterName=Core/Glow Hard Edge Toggled, sliderMax=1}
  n86 = StaticSwitchParameter(
   True=LinearInterpolate(
    A=@n83.Result
    B=MaterialFunctionCall(X=@n84, Y=@n85) {materialFunction=&MakeFloat4}.Result
    Alpha=ScalarParameter() {parameterName=ToggleThermalMask}
   )
   False=@n83.Result
  ) {group=Toggleable Reticle, parameterName=Enable Toggleable Reticle, sortPriority=5}
  NamedRerouteDeclaration(None=ComponentMask(None=@n86) {r=true}) {name=Reticle Opacity, nodeColor=<rgba 1 0.2344 0 1>}
  NamedRerouteDeclaration(None=ComponentMask(None=@n86) {g=true}) {name=Core/Glow Hard Edge, nodeColor=<rgba 1 0.2344 0 1>}
  n87 = TextureSampleParameter2D(UVs=NamedRerouteDeclaration(None=@n4.UVs) {name=CrossHairUVs, nodeColor=<rgba 0 1 0.9766 1>}) {
   group=Crosshair
   parameterName=T_Crosshair
   samplerSource=SSM_Clamp_WorldGroupSettings
   samplerType=SAMPLERTYPE_LinearColor
   texture=/Game/Weapons/FORT_Rifles/Textures/Crosshair_Dot_M
  }
  NamedRerouteDeclaration(None=@n87.RGBA) {name=Reticle Masks Pre, nodeColor=<rgba 1 0 0.4104 1>}
  n88 = VectorParameter() {
   group=Toggleable Reticle
   parameterName=DF Min Toggled
   sortPriority=100
   channelNames {a=Alpha Channel DF Min, b=Blue Channel DF Min, g=Green Channel DF Min, r=Red Channel DF Min}
   defaultValue=<rgba 0.5 0.5 0 0>
  }
  n89 = AppendVector(A=@n88, B=@n88)
  n90 = VectorParameter() {
   group=Distace Field Reticle
   parameterName=DF Min
   sortPriority=90
   channelNames {a=Alpha Channel DF Min, b=Blue Channel DF Min, g=Green Channel DF Min, r=Red Channel DF Min}
   defaultValue=<rgba 0.5 0.5 0 0>
  }
  n91 = AppendVector(A=@n90, B=@n90)
  n92 = ScalarParameter() {parameterName=ToggleThermalMask}
  n93 = VectorParameter() {
   group=Toggleable Reticle
   parameterName=DF Max Toggled
   sortPriority=101
   channelNames {a=Alpha Channel DF Max, b=Blue Channel DF Max, g=Green Channel DF Max, r=Red Channel DF Max}
   defaultValue=<rgba 0.5 0.5 0 0>
  }
  n94 = AppendVector(A=@n93, B=@n93)
  n95 = VectorParameter() {
   group=Distace Field Reticle
   parameterName=DF Max
   sortPriority=91
   channelNames {a=Alpha Channel DF Max, b=Blue Channel DF Max, g=Green Channel DF Max, r=Red Channel DF Max}
   defaultValue=<rgba 0.5 0.5 0 0>
  }
  n96 = AppendVector(A=@n95, B=@n95)
  n97 = NamedRerouteUsage()
  n98 = SmoothStep(
   Min=StaticSwitchParameter(True=LinearInterpolate(A=@n91, B=@n89, Alpha=@n92) {desc=Min DF}, False=@n91) {group=Toggleable Reticle, parameterName=Enable Toggleable Reticle, sortPriority=5}
   Max=StaticSwitchParameter(True=LinearInterpolate(A=@n96, B=@n94, Alpha=@n92) {desc=Max DF}, False=@n96) {group=Toggleable Reticle, parameterName=Enable Toggleable Reticle, sortPriority=5}
   Value=@n97
  )
  NamedRerouteDeclaration(None=StaticSwitchParameter(True=OneMinus(None=@n98), False=@n97) {group=Distace Field Reticle, parameterName=Use SDF}) {name=Reticle Masks Post, nodeColor=<rgba 0 0.1797 1 1>}
  NamedRerouteDeclaration(None=@n18.Camera - Eye Point Distance) {name=Camera - Eye point distance, nodeColor=<rgba 0 0.9766 1 1>}
  n99 = ScalarParameter() {desc="Set by Code when Player Moves and Weapon becomes \r\nless accurate 0-1 ", parameterName=SpreadRatio}
  n100 = NamedRerouteUsage()
  n101 = LinearInterpolate(
   A=Saturate(None=Divide(A=@n99) {constB=1.5})
   Alpha=Multiply(A=OneMinus(None=@n100), B=ScalarParameter() {defaultValue=1, parameterName=CrossHairBlur_FadeBlurMask})
  ) {constB=0}
  NamedRerouteDeclaration(None=LinearInterpolate(B=ScalarParameter() {defaultValue=3, group=Blur based on Accuracy, parameterName=MaxBlurMipBias}, Alpha=@n101)) {name=MipbiasBlur, nodeColor=<rgba 0 1 0.8125 1>}
  n102 = NamedRerouteUsage()
  n103 = Constant2Vector() {g=0.5, r=0.5}
  n104 = ScalarParameter() {defaultValue=1, group=Blur based on Accuracy, parameterName=CrossHairBlur_FadeMaskRadius}
  n105 = ScalarParameter() {group=Blur based on Accuracy, parameterName=CrossHairBlur_FadeMaskHarness}
  NamedRerouteDeclaration(None=Saturate(None=SphereMask(A=@n102, B=@n103, Radius=@n104, Hardness=@n105))) {name=CrosshairBlurMask, nodeColor=<rgba 0 0.9531 1 1>}
  n106 = Time()
  n107 = ScalarParameter() {defaultValue=0.2, group=Scanlines, parameterName=ScanlinesTimeMultiplier}
  NamedRerouteDeclaration(None=MaterialFunctionCall!(
    Number of Scanlines=ScalarParameter() {defaultValue=150, group=Scanlines, parameterName=Scanline Amount}
    Scanline Speed=ScalarParameter() {defaultValue=0.0085, group=Scanlines, parameterName=Scanline Speed}
    "Scanlines Thickness (Bias)"=ScalarParameter() {defaultValue=-0.09, group=Scanlines, parameterName=ScanlineSize}
    Scanlines Dark Value=ScalarParameter() {defaultValue=0.7, group=Scanlines, parameterName=Scanlines Dark}
    Scanlines Light Value=ScalarParameter() {defaultValue=1, group=Scanlines, parameterName=Scanlines Light}
    Block Opacity
    UVs
    Time=Multiply(A=@n106, B=@n107)
    Block Repeat
   ) {materialFunction=/Game/Effects/Fort_Effects/Effects/Atlas/Materials/MF_Atlas_Scanlines_01}.Result) {name=Scan_Lines, nodeColor=<rgba 0 0.8594 1 1>}
  NamedRerouteDeclaration(None=AppendVector(
    A=ScalarParameter() {defaultValue=0.5, group=Numbers, parameterName=u_NumberPaddingScale}
    B=ScalarParameter() {defaultValue=0.3, group=Numbers, parameterName=v_NumberPaddingScale}
   )) {name=padding, nodeColor=<rgba 0 1 0.6016 1>}
  n108 = VectorParameter() {group=Crosshair, parameterName=Range_Finder_Color, defaultValue=<rgba 1 0.9621 0.1678 1>}
  n109 = StaticSwitchParameter(True=NamedRerouteUsage(), False=@n108.RGB) {defaultValue=true, group=Numbers, parameterName=Calm_Rock?}
  n110 = Divide(A=ScalarParameter() {group=Numbers, parameterName=Distance}) {constB=5}
  n111 = Multiply(A=Round(None=Subtract(A=@n110) {constB=0.5})) {constB=5, desc=used to be five}
  n112 = MaterialFunctionCall!(Number=@n111, UVPadding=NamedRerouteUsage(), UVs=NamedRerouteUsage()) {materialFunction=&/MF_ADS_Distance_JethroV2}
  n113 = NamedRerouteUsage()
  n114 = ScalarParameter() {defaultValue=0.5, group=Numbers, parameterName=Error_Icon_Scale}
  n115 = Power(Base=TextureSample(UVs=MaterialFunctionCall(UVs=@n113, Texture Scale=@n114) {materialFunction=&ScaleUVsByCenter}.UVs) {desc=To show we have gone past max distance of range finder, texture=&/T_Jethro_Error}.A) {constExponent=2.5}
  n116 = ScalarParameter() {desc=Showcae out of Range symbol, group=Numbers, parameterName=OutofRange}
  n117 = Add(A=@n111) {constB=-50}
  n118 = NamedRerouteUsage()
  n119 = NamedRerouteUsage()
  n120 = VectorParameter() {group=Numbers, parameterName=Offset/Scale, channelNames {a=Offset Y, b=Offset X, g=Scale Y, r=Scale X}, defaultValue=<rgba 1 1 0 0>}
  n121 = ComponentMask(None=@n120) {g=true, r=true}
  n122 = Multiply(A=@n121, B=ScalarParameter() {defaultValue=1, group=Numbers, parameterName=secondary_Scale})
  n123 = ScalarParameter() {group=Numbers, parameterName=Rotate_Value}
  n124 = NamedRerouteUsage()
  n125 = VectorParameter() {group=Numbers, parameterName=Secondary_Number_Offset, channelNames {a="", b="", g=Y Offset, r=X Offset}, defaultValue=<rgba 0.75 0 0 1>}
  n126 = AppendVector(A=@n120.Offset X, B=@n120.Offset X)
  n127 = Add(
   A=Rotator(
    Coordinate=MaterialFunctionCall(UVs=@n119, Texture Scale=@n122) {materialFunction=&ScaleUVsByCenter}.UVs
    Time=Add(A=@n123, B=@n124)
   ) {speed=-0.15}
   B=Add(A=ComponentMask(None=@n125) {g=true, r=true}, B=@n126)
  )
  n128 = ScalarParameter() {defaultValue=1, group=Numbers, parameterName=secondary_Scale}
  n129 = ScalarParameter() {group=Numbers, parameterName=Rotate_Value}
  n130 = Rotator(
   Coordinate=MaterialFunctionCall(UVs=NamedRerouteUsage(), Texture Scale=Multiply(A=@n121, B=@n128)) {materialFunction=&ScaleUVsByCenter}.UVs
   Time=Add(A=Multiply(A=@n129) {constB=-1}, B=NamedRerouteUsage())
  ) {speed=-0.15}
  n131 = Add(A=AppendVector(A=Multiply(A=@n125.X Offset) {constB=-1}, B=@n125.Y Offset), B=@n126)
  n132 = Add(
   A=Add(
    A=LinearInterpolate(A=@n112.NumbersGreyscale, B=@n115, Alpha=@n116)
    B=MaterialFunctionCall(Number=@n117, UVPadding=@n118, UVs=@n127) {materialFunction=&MF_ADS_Distance_JethroV2}.NumbersGreyscale
   )
   B=MaterialFunctionCall(Number=Add(A=@n111) {constB=50}, UVPadding=NamedRerouteUsage(), UVs=Add(A=@n130, B=@n131)) {materialFunction=&MF_ADS_Distance_JethroV2}.NumbersGreyscale
  )
  n133 = Multiply(A=ComponentMask(None=Multiply(A=@n109, B=@n132)) {b=true, g=true, r=true}, B=NamedRerouteUsage())
  n134 = VectorParameter() {
   group=Numbers
   parameterName=Sphere_Fade_Controls
   channelNames {a=Density, b=Radius, g=Y position, r=X position}
   defaultValue=<rgba 0.5 0.5 0.5 0>
  }
  n135 = MaterialFunctionCall(UVs=NamedRerouteUsage(), CenterPosition=AppendVector(A=@n134.X position, B=@n134.X position), Radius=@n134.Radius, Density=@n134.Radius) {materialFunction=&RadialGradientExponential}
  NamedRerouteDeclaration(None=Saturate(None=Multiply(A=@n133, B=@n135.RadialGradientExponential))) {name=Range_Finder, nodeColor=<rgba 1 0.2812 0 1>}
  n136 = Frac(None=@n110)
  NamedRerouteDeclaration(None=Add(A=Multiply(A=@n136) {constB=-1}, B=ScalarParameter() {group=Numbers, parameterName=rangefinder_numbers_offset}) {constB=0.5}) {name=moving_uvs, nodeColor=<rgba 1 0.2812 0 1>}
  n137 = MaterialFunctionCall(Float2=MaterialFunctionCall() {materialFunction=&ScreenResolution}.Visible Resolution) {materialFunction=&BreakOutFloat2Components}
  n138 = Divide(A=@n137.R, B=@n137.R)
  NamedRerouteDeclaration(None=AppendVector(A=@n138, B=Constant() {r=1})) {name=Screen Ratio, nodeColor=<rgba 1 0.1 0.1 1>}
  NamedRerouteDeclaration(None=@n4.UVs) {name=CrosshairUVs_NonStatic, nodeColor=<rgba 0.4062 0 1 1>}
  n139 = ScalarParameter() {group=Noise, parameterName=NoiseAdd}
  n140 = Time()
  n141 = Multiply(A=ScalarParameter() {defaultValue=0.5, group=Noise, parameterName=FlickerSpeed}, B=@n140)
  n142 = AppendVector(A=AppendVector(A=Sine(None=@n141) {period=1.6}, B=Sine(None=@n141) {period=1.42}), B=Sine(None=@n141) {period=1.51})
  n143 = ScalarParameter() {defaultValue=0.5, group=Noise, parameterName=NoiseFrequency}
  n144 = Add(A=Sine(None=@n141) {period=2.1}, B=Multiply(B=@n143) {constA=0.25})
  n145 = Add(A=Sine(None=@n141) {period=0.72}, B=Multiply(B=@n143) {constA=0.75}) {constB=0.75}
  n146 = Ceil(None=Add(A=Multiply(B=@n143) {constA=0.9, constB=0.9}, B=Sine(None=@n141) {period=1.25}) {constB=0.9})
  n147 = Saturate(None=Multiply(A=Multiply(A=@n144, B=@n145), B=Multiply(A=@n145, B=@n146)))
  n148 = LinearInterpolate(Alpha=ConstantBiasScale(None=Sine(None=@n140) {period=0.075})) {constA=0.7}
  n149 = Multiply(A=Multiply(A=Ceil(None=@n142), B=Multiply(A=@n147, B=@n148)), B=ScalarParameter() {defaultValue=2, group=Noise, parameterName=NoiseMult})
  NamedRerouteDeclaration(None=Saturate(None=Add(A=@n139, B=@n149))) {name=flicker_Speed, nodeColor=<rgba 1 0.5391 0 1>}
  NamedRerouteDeclaration(None=@n108.RGB) {name=range finder color, nodeColor=<rgba 0.9688 1 0 1>}
  NamedRerouteDeclaration(None=@n132) {name=Number Mask, nodeColor=<rgba 0 1 0.4844 1>}
  n150 = MaterialFunctionCall(Float2=Add(A=ScreenPosition().ViewportUV) {constB=-0.5}) {materialFunction=&BreakOutFloat2Components}
  n151 = MaterialFunctionCall(X=Multiply(A=@n150.R, B=@n138), Y=@n150.G) {materialFunction=&MakeFloat2}
  NamedRerouteDeclaration(None=Add(A=NamedRerouteDeclaration(None=@n151.Result) {name=Screen Position Centered, nodeColor=<rgba 1 0.1 0.1 1>}) {constB=0.5}) {name=Screen Position Ratioed, nodeColor=<rgba 1 0.1 0.1 1>}
  n152 = MaterialFunctionCall(UVs=NamedRerouteUsage(), Texture Scale=@n121) {materialFunction=&ScaleUVsByCenter}
  n153 = NamedRerouteUsage()
  NamedRerouteDeclaration(None=Add(A=Rotator(Coordinate=@n152.UVs, Time=@n153) {speed=-0.15}, B=@n126)) {name=Center Number UV, nodeColor=<rgba 0.2656 1 0 1>}
  n154 = NamedRerouteUsage()
  n155 = Constant2Vector() {g=1}
  n156 = Add(
   A=Subtract(A=DotProduct(A=@n154, B=@n155))
   B=StaticSwitchParameter(
    True=MaterialFunctionCall() {materialFunction=&DebugTimeSine}.Result
    False=ScalarParameter() {defaultValue=0.5, group=Overheat, parameterName=Progress}
   ) {group=Overheat, parameterName=Debug?}
  )
  NamedRerouteDeclaration(None=Saturate(None=Multiply(A=@n156) {constB=256})) {name=DerivedHeatPercent, nodeColor=<rgba 0 1 0.5781 1>}
  n157 = NamedRerouteUsage()
  n158 = ScalarParameter() {defaultValue=-2, group=Overheat, parameterName=Overheat_dark_background}
  n159 = VectorParameter() {group=Overheat, parameterName=Glow_Color_Overheat, defaultValue=<rgba 0.724 0.0912 0.0912 0>}
  n160 = Multiply(A=@n159.RGB, B=@n159.RGB)
  n161 = VectorParameter() {group=Overheat, parameterName=Glow_ColorB_Overheat, defaultValue=<rgba 0.724 0.0912 0.0912 0>}
  n162 = Multiply(A=@n161.RGB, B=LinearInterpolate(A=@n161.A, Alpha=NamedRerouteUsage()) {constB=0.5})
  n163 = NamedRerouteUsage()
  n164 = NamedRerouteUsage()
  n165 = TextureSample(UVs=NamedRerouteUsage()) {samplerSource=SSM_Clamp_WorldGroupSettings, texture=&/Overheat}
  n166 = Multiply(
   A=LinearInterpolate(A=Multiply(A=@n157, B=@n158) {constB=-10}, B=LinearInterpolate(A=@n160, B=@n162, Alpha=@n163), Alpha=Multiply(A=@n164, B=@n165.R))
   B=@n165.R
  )
  n167 = Multiply(A=@n161.RGB, B=@n161.RGB)
  NamedRerouteDeclaration(None=Multiply(A=LinearInterpolate(A=@n166, B=@n167, Alpha=@n165.B), B=NamedRerouteUsage())) {name=Overheat_Bar, nodeColor=<rgba 0.0078 1 0 1>}
  n168 = NamedRerouteUsage()
  n169 = VectorParameter() {
   group=Overheat
   parameterName=Offset/Scale_Overheat
   channelNames {a=Offset Y, b=Offset X, g=Scale Y, r=Scale X}
   defaultValue=<rgba 1 1 0 0>
  }
  n170 = ComponentMask(None=@n169) {g=true, r=true}
  NamedRerouteDeclaration(None=Add(
    A=MaterialFunctionCall(UVs=@n168, Texture Scale=@n170) {materialFunction=&ScaleUVsByCenter}.UVs
    B=AppendVector(A=@n169.Offset X, B=@n169.Offset X)
   )) {name=Overheat_UV's, nodeColor=<rgba 0 1 0.8594 1>}
  n171 = ScalarParameter() {defaultValue=0.5, group=Overheat, parameterName=Progress}
  n172 = MaterialFunctionCall(Input=@n171, Input Low=Constant() {r=0.37}, Input High=Constant() {r=0.7}) {materialFunction=&RemapValue_0-1}
  n173 = Constant()
  NamedRerouteDeclaration(None=Saturate(None=If(A=@n171, "A > B"=@n172.Result, "A == B"=@n172.Result, "A < B"=@n173) {constB=0.37})) {name=Heat_Scale, nodeColor=<rgba 0 1 0.1328 1>}
  n174 = Sine(None=Multiply(A=Time()) {constB=3})
  n175 = ScalarParameter() {defaultValue=0.5, group=Overheat, parameterName=Progress}
  n176 = ScalarParameter() {defaultValue=0.85, desc=Percentage Gauge starts flashing, group=Overheat, parameterName=Flicker Point}
  n177 = Constant() {r=1}
  n178 = Constant()
  NamedRerouteDeclaration(None=LinearInterpolate(B=ConstantBiasScale(None=@n174), Alpha=If(A=@n175, B=@n176, "A > B"=@n177, "A == B"=@n177, "A < B"=@n178))) {name=Flicker Control, nodeColor=<rgba 1 0.2578 0 1>}
  n179 = SceneTexture(UVs=ScreenPosition().ViewportUV) {bFiltered=true, sceneTextureId=PPI_PostProcessInput0}
  n180 = ComponentMask(None=SceneTexture(UVs=NamedRerouteUsage()) {bFiltered=true, sceneTextureId=PPI_PostProcessInput0}.Color) {b=true, g=true, r=true}
  n181 = SphereMask(A=NamedRerouteUsage(), B=Constant() {r=1}) {attenuationRadius=0.2, hardnessPercent=0}
  n182 = NamedRerouteUsage()
  NamedRerouteDeclaration(None=LinearInterpolate(A=ComponentMask(None=@n179.Color) {b=true, g=true, r=true}, B=@n180, Alpha=Multiply(A=@n181, B=@n182))) {name=Scene Texture, nodeColor=<rgba 0 1 0.1797 1>}
  NamedRerouteDeclaration(None=Saturate(None=ScalarParameter() {desc=Set in C++, parameterName=TargetingWeight})) {name=ADS Transition, nodeColor=<rgba 1 0 0.5391 1>}
  NamedRerouteDeclaration(None=MaterialFunctionCall(UVs=ScreenPosition().ViewportUV, Texture Scale=@n21) {materialFunction=&ScaleUVsByCenter}.UVs) {name=ModifiedScreenPosition, nodeColor=<rgba 0 1 0.8594 1>}
  NamedRerouteUsage()
  NamedRerouteUsage()
  NamedRerouteUsage()
  NamedRerouteUsage()
  NamedRerouteUsage()
  NamedRerouteUsage()
  NamedRerouteUsage()
  Rotator(Coordinate=NamedRerouteUsage(), Time=NamedRerouteUsage()) {speed=-0.15}
  n183 = Time() {bOverride_Period=true, period=5}
  n184 = Sine(None=@n183) {period=3}
  n185 = Sine(None=@n183) {period=1.5}
  Saturate(None=AppendVector(A=AppendVector(A=@n184, B=@n185), B=Sine(None=@n183) {period=1.2}))
  ScalarParameter() {group=Numbers, parameterName=Distance_DEbug}
  StaticBool() {value=true}
  StaticSwitchParameter(True=NamedRerouteUsage(), False=Multiply(A=NamedRerouteUsage(), B=NamedRerouteUsage())) {group=Debug, parameterName=DebugView, sortPriority=5}
  n186 = Divide(A=ComponentMask(None=ViewProperty() {property=MEVP_ViewSize}.Property) {g=true}) {constB=2000}
  n187 = Constant2Vector() {g=1, r=0.5}
  n188 = InverseLinearInterpolate(Value=ComponentMask(None=SceneTexture(UVs=@n187) {sceneTextureId=PPI_SceneDepth}.Color) {r=true}) {constA=10, constB=30}
  n189 = ScreenPosition()
  n190 = Saturate(None=Subtract(A=ComponentMask(None=@n189.ViewportUV) {g=true}) {constB=0.65})
  n191 = Power(Base=Sine(None=ComponentMask(None=@n189.ViewportUV) {r=true}) {period=2})
  n192 = LinearInterpolate(
   A=ScalarParameter() {defaultValue=0.001, group="DOF Blur (SM6 Only)", parameterName=Blur_DistanceMin}
   B=LinearInterpolate(
    A=ScalarParameter() {defaultValue=0.001, group="DOF Blur (SM6 Only)", parameterName=Blur_DistanceMax}
    B=ScalarParameter() {defaultValue=0.001, group="DOF Blur (SM6 Only)", parameterName=Blur_DistanceMaxPistols}
    Alpha=Saturate(None=@n188)
   )
   Alpha=NamedRerouteDeclaration(None=Multiply(A=@n190, B=@n191)) {name=Blur Gradient, nodeColor=<rgba 1 0 0.2812 1>}
  )
  MaterialExpressionMultiply_3 = Multiply(A=OneMinus(None=NamedRerouteUsage()), B=Multiply(A=Multiply(A=@n186, B=@n192), B=NamedRerouteUsage()))
  MaterialExpressionRound_0 = Round(None=ScalarParameter() {defaultValue=3, group="DOF Blur (SM6 Only)", parameterName=Blur Samples})
  MaterialExpressionMaterialFunctionCall_1 = MaterialFunctionCall() {materialFunction=&DitherTemporalAA}
  MaterialExpressionScalarParameter_7 = ScalarParameter() {defaultValue=1, group="DOF Blur (SM6 Only)", parameterName=TempAARotation}
  MaterialExpressionScalarParameter_8 = ScalarParameter() {defaultValue=4, group="DOF Blur (SM6 Only)", parameterName=DistanceMask}
  MaterialExpressionScalarParameter_14 = ScalarParameter() {defaultValue=1, group="DOF Blur (SM6 Only)", parameterName=TempAADistance}
  MaterialExpressionScalarParameter_15 = ScalarParameter() {defaultValue=0.618, group="DOF Blur (SM6 Only)", parameterName=Radial Offset}
  MaterialExpressionNamedRerouteUsage_15 = NamedRerouteUsage()
  MaterialExpressionSceneTexture_3 = SceneTexture() {sceneTextureId=PPI_PostProcessInput0}
  n193 = Custom!(
   Distance=@MaterialExpressionMultiply_3
   Steps=@MaterialExpressionRound_0
   Dither=@MaterialExpressionMaterialFunctionCall_1.Result
   TempAARotation=@MaterialExpressionScalarParameter_7
   DistanceMask=@MaterialExpressionScalarParameter_8
   TempAADistance=@MaterialExpressionScalarParameter_14
   RadialOffset=@MaterialExpressionScalarParameter_15
   UV=@MaterialExpressionNamedRerouteUsage_15
   SceneTexture=@MaterialExpressionSceneTexture_3.Color
  ) {
   code="float3 CurColor = .0;\r\nfloat RandomSamp = Dither; //Input\r\nfloat2 BaseUV = ClampSceneTextureUV(ViewportUVToSceneTextureUV(DERIV_BASE_VALUE(UV), 14), 14);\r\nfloat2 NewUV = BaseUV; //Input\r\nfloat CurDistance = .0;\r\nfloat2 CurOffset = .0;\r\n\r\n//int DistanceSteps = Steps + (int) RandomSamp;\r\n//int RadialSteps = Steps + (1 - (int) RandomSamp);\r\nint DistanceSteps = Steps;\r\nint RadialSteps = Steps;\r\nfloat StepSize = Distance / (int) DistanceSteps;\r\n\r\nfloat TwoPi = 6.283185;\r\nfloat Substep = .0;\r\nfloat2 ScenePixels=View.BufferSizeAndInvSize.xy*BaseUV;\r\n\r\n\r\nTempAARotation *= RandomSamp;\r\nTempAADistance *= StepSize * RandomSamp;\r\nfloat tempdist = -.0;\r\nfloat totaldist = .0;\r\n\r\nint i=0;\r\nif (DistanceSteps<1 || Distance <= .0)\r\n{\r\nreturn SceneTextureLookup(BaseUV, 14, 1).rgb;\r\n}\r\nelse\r\n{\r\nwhile ( i < (int) DistanceSteps)\r\n{\r\n\r\n//CurDistance+=StepSize;\r\nfor (int j = 0; j < (int) RadialSteps; j++)\r\n{\r\nCurOffset.x = cos(TwoPi*((TempAARotation+Substep) / RadialSteps));\r\nCurOffset.y = sin(TwoPi*((TempAARotation+Substep) / RadialSteps));\r\nCurOffset *=DistanceMask;\r\nNewUV.x = BaseUV.x + (CurOffset.x * (CurDistance+(RandomSamp*TempAADistance)));\r\nNewUV.y = BaseUV.y + (CurOffset.y * (CurDistance+(RandomSamp*TempAADistance)));\r\n\t\ttempdist = (length(NewUV-BaseUV)/Distance);\r\n\t\ttotaldist += tempdist;\r\n\r\nCurColor += SceneTextureLookup(NewUV, 14, 1).rgb*tempdist;\r\n\r\nSubstep++;\r\n}\r\nCurDistance+=StepSize;\r\nSubstep+=RadialOffset;\r\ni++;\r\n}\r\nCurColor = CurColor / totaldist;\r\nreturn CurColor;\r\n}"
   desc="\n*** Original source before expression upgrade ***\nfloat3 CurColor;\r\nfloat2 BaseUV = MaterialFloat2(ScreenAlignedPosition(Parameters.ScreenPosition).xy);\r\nfloat2 NewUV = BaseUV;\r\nfloat StepSize = Distance / (int) DistanceSteps;\r\nfloat CurDistance=0;\r\nfloat2 CurOffset;\r\nfloat TwoPi = 6.283185;\r\nfloat Substep;\r\nfloat2 ScenePixels=View.RenderTargetSize*BaseUV;\r\nScenePixels+=View.TemporalAAParams.r;\r\nfloat2 RandomSamp = ((uint)(ScenePixels.x) + 2 * (uint)(ScenePixels.y)) % 5;\r\nRandomSamp+=Texture2DSample(Tex,TexSampler,ScenePixels);\r\nRandomSamp/=5;\r\nRandomSamp-=0.5;\r\nTempAARotation*=RandomSamp;\r\nTempAADistance*=StepSize*RandomSamp;\r\nfloat tempdist = -0;\r\nfloat totaldist = 0;\r\n\r\nint i=0;\r\nif (DistanceSteps<1)\r\n{\r\nreturn DecodeSceneColorForMaterialNode(NewUV);\r\n}\r\nelse\r\n{\r\n//CurDistance += 0.5*StepSize;\r\nwhile ( i < (int) DistanceSteps)\r\n{\r\n\r\n//CurDistance+=StepSize;\r\nfor (int j = 0; j < (int) RadialSteps; j++)\r\n{\r\nCurOffset.x = cos(TwoPi*((TempAARotation+Substep) / RadialSteps));\r\nCurOffset.y = sin(TwoPi*((TempAARotation+Substep) / RadialSteps));\r\nCurOffset *=DistanceMask;\r\nNewUV.x = BaseUV.x + (CurOffset.x * (CurDistance+(RandomSamp*TempAADistance)));\r\nNewUV.y = BaseUV.y + (CurOffset.y * (CurDistance+(RandomSamp*TempAADistance)));\r\n\t\ttempdist = (length(NewUV-BaseUV)/Distance);\r\n\t\ttotaldist += tempdist;\r\n\r\nCurColor += DecodeSceneColorForMaterialNode(NewUV)*tempdist;\r\n//CurDistance+=(StepSize+(TempAADistance))/RadialSteps;\r\nSubstep++;\r\n}\r\nCurDistance+=StepSize;\r\nSubstep+=RadialOffset;\r\ni++;\r\n}\r\nCurColor = CurColor /totaldist;\r\nreturn CurColor;\r\n}\n*** Original source before expression upgrade ***\nfloat3 CurColor;\r\nfloat2 BaseUV = MaterialFloat2(ScreenAlignedPosition(Parameters.ScreenPosition).xy);\r\nfloat2 NewUV = BaseUV;\r\nfloat StepSize = Distance / (int) DistanceSteps;\r\nfloat CurDistance=0;\r\nfloat2 CurOffset;\r\nfloat TwoPi = 6.283185;\r\nfloat Substep;\r\nfloat2 ScenePixels=Frame.RenderTargetSize*BaseUV;\r\nScenePixels+=Frame.TemporalAAParams.r;\r\nfloat2 RandomSamp = ((uint)(ScenePixels.x) + 2 * (uint)(ScenePixels.y)) % 5;\r\nRandomSamp+=Texture2DSample(Tex,TexSampler,ScenePixels);\r\nRandomSamp/=5;\r\nRandomSamp-=0.5;\r\nTempAARotation*=RandomSamp;\r\nTempAADistance*=StepSize*RandomSamp;\r\nfloat tempdist = -0;\r\nfloat totaldist = 0;\r\n\r\nint i=0;\r\nif (DistanceSteps<1)\r\n{\r\nreturn DecodeSceneColorForMaterialNode(NewUV);\r\n}\r\nelse\r\n{\r\n//CurDistance += 0.5*StepSize;\r\nwhile ( i < (int) DistanceSteps)\r\n{\r\n\r\n//CurDistance+=StepSize;\r\nfor (int j = 0; j < (int) RadialSteps; j++)\r\n{\r\nCurOffset.x = cos(TwoPi*((TempAARotation+Substep) / RadialSteps));\r\nCurOffset.y = sin(TwoPi*((TempAARotation+Substep) / RadialSteps));\r\nCurOffset *=DistanceMask;\r\nNewUV.x = BaseUV.x + (CurOffset.x * (CurDistance+(RandomSamp*TempAADistance)));\r\nNewUV.y = BaseUV.y + (CurOffset.y * (CurDistance+(RandomSamp*TempAADistance)));\r\n\t\ttempdist = (length(NewUV-BaseUV)/Distance);\r\n\t\ttotaldist += tempdist;\r\n\r\nCurColor += DecodeSceneColorForMaterialNode(NewUV)*tempdist;\r\n//CurDistance+=(StepSize+(TempAADistance))/RadialSteps;\r\nSubstep++;\r\n}\r\nCurDistance+=StepSize;\r\nSubstep+=RadialOffset;\r\ni++;\r\n}\r\nCurColor = CurColor /totaldist;\r\nreturn CurColor;\r\n}"
   description=SceneTextureBlur
   inputs [
    {inputName=Distance, input {expression=&self:MaterialExpressionMultiply_3}}
    {inputName=Steps, input {expression=&self:MaterialExpressionRound_0}}
    {inputName=Dither, input {expression=&self:MaterialExpressionMaterialFunctionCall_1}}
    {inputName=TempAARotation, input {expression=&self:MaterialExpressionScalarParameter_7}}
    {inputName=DistanceMask, input {expression=&self:MaterialExpressionScalarParameter_8}}
    {inputName=TempAADistance, input {expression=&self:MaterialExpressionScalarParameter_14}}
    {inputName=RadialOffset, input {expression=&self:MaterialExpressionScalarParameter_15}}
    {inputName=UV, input {expression=&self:MaterialExpressionNamedRerouteUsage_15}}
    {
     inputName=SceneTexture
     input {
      expression=&self:MaterialExpressionSceneTexture_3
      mask=1
      maskA=1
      maskB=1
      maskG=1
      maskR=1
     }
    }
   ]
  }
  StaticSwitchParameter(True=FeatureLevelSwitch(Default=@n180, SM6=@n193), False=@n180) {group="DOF Blur (SM6 Only)", parameterName=Enable DOF Blur, sortPriority=10}
  n194 = TextureSampleParameter2D!(UVs=NamedRerouteUsage(), Bias=NamedRerouteUsage(), Apply View MipBias) {
   group=Crosshair
   mipValueMode=TMVM_MipBias
   parameterName=T_Crosshair
   samplerSource=SSM_Clamp_WorldGroupSettings
   samplerType=SAMPLERTYPE_LinearColor
   texture=&Crosshair_Dot_M
  }
  n195 = NamedRerouteUsage()
  n196 = MaterialFunctionCall!(OffsetAmount=LinearInterpolate(B=ScalarParameter() {defaultValue=0.001, group=Blur based on Accuracy, parameterName=MaxBlurSampleOffset}, Alpha=@n101)) {materialFunction=/Engine/Functions/Engine_MaterialFunctions01/ImageAdjustment/BlurSampleOffsets}
  n197 = TextureSampleParameter2D(UVs=Add(A=@n195, B=@n196."-.5,.5"), Bias=NamedRerouteUsage()) {
   group=Crosshair
   mipValueMode=TMVM_MipBias
   parameterName=T_Crosshair
   samplerSource=SSM_Clamp_WorldGroupSettings
   samplerType=SAMPLERTYPE_LinearColor
   texture=&Crosshair_Dot_M
  }
  n198 = Add(A=@n194.RGBA, B=@n197.RGBA)
  n199 = TextureSampleParameter2D(UVs=Add(A=@n195, B=@n196.".5,.5"), Bias=NamedRerouteUsage()) {
   group=Crosshair
   mipValueMode=TMVM_MipBias
   parameterName=T_Crosshair
   samplerSource=SSM_Clamp_WorldGroupSettings
   samplerType=SAMPLERTYPE_LinearColor
   texture=&Crosshair_Dot_M
  }
  n200 = TextureSampleParameter2D(UVs=Add(A=@n196.".5,-.5", B=@n195), Bias=NamedRerouteUsage()) {
   group=Crosshair
   mipValueMode=TMVM_MipBias
   parameterName=T_Crosshair
   samplerSource=SSM_Clamp_WorldGroupSettings
   samplerType=SAMPLERTYPE_LinearColor
   texture=&Crosshair_Dot_M
  }
  n201 = TextureSampleParameter2D(UVs=Add(A=@n196."-.5,-.5", B=@n195), Bias=NamedRerouteUsage()) {
   group=Crosshair
   mipValueMode=TMVM_MipBias
   parameterName=T_Crosshair
   samplerSource=SSM_Clamp_WorldGroupSettings
   samplerType=SAMPLERTYPE_LinearColor
   texture=&Crosshair_Dot_M
  }
  n202 = Divide(A=Add(A=Add(A=@n198, B=@n199.RGBA), B=Add(A=@n200.RGBA, B=@n201.RGBA))) {constB=5}
  n203 = Multiply(A=@n202, B=ScalarParameter() {defaultValue=1, group=Blur based on Accuracy, parameterName=CrossHairBlur_FadeOpacity})
  n204 = NamedRerouteUsage()
  StaticSwitchParameter(True=LinearInterpolate(A=@n202, B=LinearInterpolate(A=@n202, B=@n203, Alpha=@n204), Alpha=@n101), False=@n87.RGBA) {defaultValue=true, parameterName=EnableCrossHair Blur}
  n205 = NamedRerouteUsage()
  n206 = NamedRerouteUsage()
  n207 = VectorParameter() {
   group=Numbers
   parameterName=Highlight_Sphere
   channelNames {a=Density, b=Radius, g=Y position, r=X position}
   defaultValue=<rgba 0.5 0.5 0.5 0>
  }
  n208 = MaterialFunctionCall(UVs=NamedRerouteUsage(), CenterPosition=AppendVector(A=@n207.X position, B=@n207.X position), Radius=@n207.Radius, Density=@n207.Radius) {materialFunction=&RadialGradientExponential}
  n209 = NamedRerouteUsage()
  n210 = Saturate(None=Subtract(A=Multiply(A=@n206, B=@n208.RadialGradientExponential), B=Multiply(A=@n209) {constB=0.5}))
  n211 = ScalarParameter() {defaultValue=1, group=Numbers, parameterName=highlight_Sphere_Multi}
  StaticSwitchParameter(True=@n47, False=LinearInterpolate(A=@n47, B=ComponentMask(None=@n205) {b=true, g=true, r=true}, Alpha=Multiply(A=@n210, B=@n211))) {defaultValue=true, group=Numbers, parameterName=Calm_Rock?}
  ViewProperty() {property=MEVP_PostVolumeUserFlags}
 ]
}
